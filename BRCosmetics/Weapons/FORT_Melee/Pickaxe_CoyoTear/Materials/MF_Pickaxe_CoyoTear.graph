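# Material function for the CoyoTear pickaxe that adds an animated, glowing blade effect to incoming material attributes.
# Radial UVs from local position pan mask-texture noise, which is layered and remapped through a curve-atlas gradient for blade color.
# Similar radial noise drives masked world position offset, with masks built from the TechArtMask input, pre-skinned position and normals.
# Color_Time_of_Day scales the glow, which feeds emissive and adds to base color.
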MaterialFunction /BRCosmetics/Weapons/FORT_Melee/Pickaxe_CoyoTear/Materials/MF_Pickaxe_CoyoTear
graph {nodes [
  n1 = GetMaterialAttributes(None=FunctionInput() {inputType=FunctionInput_MaterialAttributes}) {attributeGetTypes [#1, #2]}
  n2 = ComponentMask(None=NamedRerouteUsage()) {g=true}
  n3 = Add(
   A=Multiply(A=NamedRerouteDeclaration(None=@n1.EmissiveColor) {name=EmissiveColor, nodeColor=<rgba 1 0 0.375 1>}, B=OneMinus(None=@n2))
   B=NamedRerouteUsage()
  )
  n4 = ScalarParameter() {defaultValue=0.5, parameterName=EmissiveRatio}
  n5 = Multiply(A=@n3, B=@n4) {constB=0.5}
  n6 = MaterialFunctionCall(Color_Morning=@n5, Color_Day=@n5, Color_Evening=@n5, Color_Night=Multiply(A=@n3, B=Multiply(A=@n4) {constB=0.6})) {materialFunction=&Color_Time_of_Day}
  n7 = Multiply(A=@n6.Result, B=ScalarParameter() {defaultValue=0.5, parameterName=BaseColorRatio}) {constB=0.5}
  FunctionOutput(None=SetMaterialAttributes!(
    MaterialAttributes=@n1.MaterialAttributes
    BaseColor=Add(A=@n1.BaseColor, B=@n7)
    EmissiveColor=@n6.Result
    WorldPositionOffset=NamedRerouteUsage()
   ) {attributeSetTypes [#1, #2, #3]})
  n8 = Transform()
  n9 = ComponentMask(None=VertexInterpolator(VS=PreSkinnedPosition()).PS) {r=true}
  n10 = Multiply(A=@n9) {constB=0.5}
  n11 = Transform(None=AppendVector(A=Clamp(None=@n10) {minDefault=-1}, B=Constant2Vector())) {transformSourceType=TRANSFORMSOURCE_Local}
  n12 = ScalarParameter() {defaultValue=0.5}
  LinearInterpolate(A=@n8, B=Normalize(VectorInput=LinearInterpolate(A=@n8, B=@n11, Alpha=@n12)))
  MaterialFunctionCall() {materialFunction=&LocalPosition}
  n13 = MaterialFunctionCall() {materialFunction=&LocalPosition}
  n14 = VectorParameter() {parameterName=PositionObjectCenter, defaultValue=<rgba -12.0251 -40.72 0 1>}
  n15 = Multiply(A=Add(A=ComponentMask(None=@n13.Local Position) {b=true, g=true}, B=ComponentMask(None=@n14.RGB) {g=true, r=true})) {constB=0.05}
  NamedRerouteDeclaration(None=MaterialFunctionCall(Vector or UVs=NamedRerouteDeclaration(None=@n15) {name=UV, nodeColor=<rgba 1 0 0.7969 1>}) {materialFunction=&VectorToRadialValue}.Radial Coordinates) {name=Radial UV, nodeColor=<rgba 1 0.5859 0 1>}
  n16 = MaterialFunctionCall() {materialFunction=&LocalPosition}
  n17 = ComponentMask(None=@n16.Local Position) {b=true, g=true}
  n18 = ComponentMask(None=MaterialFunctionCall() {materialFunction=&LocalPosition}.Local Position) {b=true, g=true}
  n19 = ComponentMask(None=Add(
    A=VectorParameter() {parameterName=PositionObjectCenter, defaultValue=<rgba -12.0251 -40.72 0 1>}.RGB
    B=VectorParameter() {parameterName=PositionWPOOffset, defaultValue=<rgba -45.8868 0.8809 0 1>}.RGB
   )) {g=true, r=true}
  n20 = MaterialFunctionCall(Vector or UVs=Multiply(A=Add(A=@n18, B=@n19), B=ScalarParameter() {defaultValue=0.025, parameterName=WPOTextureSize}) {constB=0.025}) {materialFunction=&VectorToRadialValue}
  n21 = ComponentMask(None=VectorParameter() {parameterName=WPOTextureSpeed, defaultValue=<rgba -0.1 -0.35 0 1>}.RGB) {g=true, r=true}
  n22 = NamedRerouteUsage()
  n23 = TextureSample!(
   UVs=Add(A=Panner(Coordinate=@n20.Radial Coordinates, Speed=@n21) {speedX=-0.1, speedY=-0.35}, B=Multiply(A=@n22) {constB=0.2})
   Tex
   Level
   Apply View MipBias
  ) {
   mipValueMode=TMVM_MipLevel
   samplerType=SAMPLERTYPE_LinearColor
   texture=/BRCosmetics/Weapons/FORT_Melee/Pickaxe_CoyoTear/Textures/T_Pickaxe_Coyotear_Mask
  }
  n24 = Multiply(A=@n23.G) {constB=2}
  n25 = Multiply(
   A=NamedRerouteDeclaration(None=Subtract(A=@n24)) {name=WPOTexture, nodeColor=<rgba 0.0078 1 0 1>}
   B=ScalarParameter() {defaultValue=1, parameterName=WPODistortionIntensity}
  ) {constB=2}
  n26 = Subtract(
   A=AppendVector(A=ComponentMask(None=@n16.Local Position) {r=true}, B=Add(A=@n17, B=@n25))
   B=MaterialFunctionCall() {materialFunction=&LocalPosition}.Local Position
  )
  n27 = NamedRerouteUsage()
  n28 = NamedRerouteUsage()
  n29 = Power(Base=Subtract(A=ComponentMask(None=@n28) {g=true}))
  n30 = NamedRerouteUsage()
  n31 = Subtract(A=Saturate(None=Multiply(A=@n29) {constB=0.5}), B=OneMinus(None=ComponentMask(None=@n30) {g=true}))
  n32 = Saturate(None=Add(A=ComponentMask(None=@n27) {b=true}, B=Saturate(None=@n31)))
  NamedRerouteDeclaration(None=Transform(None=Multiply(A=@n26, B=@n32)) {transformSourceType=TRANSFORMSOURCE_Local}) {name=WPO, nodeColor=<rgba 0.5703 1 0 1>}
  NamedRerouteDeclaration(None=FunctionInput(Preview=TextureSample!(UVs, Tex, Apply View MipBias) {texture=/BRCosmetics/Weapons/FORT_Melee/Pickaxe_CoyoTear/Textures/T_Pickaxe_CoyoTear_FX}.RGB) {inputName=TechArtMask}) {name=TechArtMask, nodeColor=<rgba 1 0.375 0 1>}
  n33 = TextureSample!(UVs=Panner(Coordinate=NamedRerouteUsage()), Tex, Level, Apply View MipBias) {
   mipValueMode=TMVM_MipLevel
   samplerType=SAMPLERTYPE_LinearColor
   texture=&T_Pickaxe_Coyotear_Mask
  }
  NamedRerouteDeclaration(None=@n33.G) {name=Radial Cloud Noise, nodeColor=<rgba 0.7344 1 0 1>}
  n34 = NamedRerouteUsage()
  n35 = AppendVector(A=ComponentMask(None=@n34) {r=true}, B=Power(Base=ComponentMask(None=@n34) {g=true}) {constExponent=0.5})
  n36 = Multiply(A=NamedRerouteUsage()) {constB=0.25}
  n37 = TextureSample(UVs=Add(A=Panner(Coordinate=@n35) {speedY=-0.1}, B=@n36)) {
   mipValueMode=TMVM_MipLevel
   samplerType=SAMPLERTYPE_LinearColor
   texture=&T_Pickaxe_Coyotear_Mask
  }
  n38 = NamedRerouteUsage()
  n39 = ComponentMask(None=NamedRerouteUsage()) {g=true}
  n40 = OneMinus(None=@n39)
  n41 = Multiply(A=Power(Base=Add(A=@n40) {constB=1.2})) {constB=0.2}
  n42 = Power(Base=Saturate(None=@n39))
  n43 = TextureSample(UVs=Add(A=Panner(Coordinate=@n35) {speedY=-0.25}, B=@n36)) {
   mipValueMode=TMVM_MipLevel
   samplerType=SAMPLERTYPE_LinearColor
   texture=&T_Pickaxe_Coyotear_Mask
  }
  n44 = Multiply(A=Multiply(A=Multiply(A=@n42, B=@n43.R)) {constB=0.5}, B=@n38)
  n45 = Add(A=Multiply(A=Multiply(A=@n37.R) {constB=0.3}, B=@n38), B=Saturate(None=Subtract(A=@n41, B=@n44)))
  n46 = NamedRerouteUsage()
  n47 = NamedRerouteUsage()
  n48 = TextureSample(UVs=Add(A=Panner(Coordinate=@n46) {speedY=-0.2}, B=Multiply(A=@n47) {constB=0.2})) {
   mipValueMode=TMVM_MipLevel
   samplerType=SAMPLERTYPE_LinearColor
   texture=&T_Pickaxe_Coyotear_Mask
  }
  n49 = OneMinus(None=ComponentMask(None=NamedRerouteUsage()) {g=true})
  n50 = Step(X=Multiply(A=@n48.R, B=Subtract(A=@n49) {constB=0.2})) {constY=0.02}
  n51 = OneMinus(None=ComponentMask(None=NamedRerouteUsage()) {g=true})
  n52 = Step(X=Multiply(A=@n43.R, B=Subtract(A=@n51) {constB=0.4})) {constY=0.05}
  n53 = Add(A=Multiply(A=Saturate(None=@n45)) {constB=0.5}, B=LinearInterpolate(A=Multiply(A=@n50) {constB=0.2}, B=@n52, Alpha=@n52))
  n54 = CurveAtlasRowParameter(CurveTime=OneMinus(None=Saturate(None=@n53))) {
   atlas=/Game/Weapons/FORT_Melee/Pickaxe_CoyoTear/FX/CA_Pickaxe_CoyoTear_Spline_Color
   curve=/Game/Weapons/FORT_Melee/Pickaxe_CoyoTear/FX/CV_Pickaxe_CoyoTear_Spline_ColorGradient
   defaultValue=1
   parameterName=Color Gradient Curve
  }
  n55 = StaticSwitchParameter(True=NamedRerouteUsage(), False=@n54) {parameterName=DebugWPOTexture}
  n56 = NamedRerouteUsage()
  n57 = NamedRerouteUsage()
  n58 = ComponentMask(None=NamedRerouteUsage()) {g=true}
  n59 = Saturate(None=Multiply(A=OneMinus(None=@n58)) {constB=3})
  n60 = Multiply(A=Saturate(None=Subtract(A=@n57, B=@n59))) {constB=0.5}
  n61 = NamedRerouteUsage()
  NamedRerouteDeclaration(None=LinearInterpolate(B=LinearInterpolate(A=@n55, B=@n56, Alpha=@n60), Alpha=ComponentMask(None=@n61) {g=true})) {name=Axe Blade Color, nodeColor=<rgba 1 0 0.1172 1>}
  n62 = ComponentMask(None=NamedRerouteUsage()) {g=true}
  n63 = Multiply(A=Power(Base=Subtract(A=@n62))) {constB=0.5}
  n64 = VertexInterpolator(VS=PreSkinnedPosition())
  n65 = Constant3Vector() {constant=<rgba 0 -14 -41 0>}
  n66 = ComponentMask(None=Divide(A=Add(A=@n64.PS, B=@n65), B=Constant3Vector() {constant=<rgba 1 32 19 0>})) {b=true, g=true}
  n67 = ComponentMask(None=@n66) {r=true}
  n68 = ComponentMask(None=@n66) {g=true}
  n69 = Multiply(A=Power(Base=OneMinus(None=@n67)) {constExponent=1.5}, B=Power(Base=Abs(None=@n68)) {constExponent=0.9})
  n70 = Transform(None=VertexNormalWS()) {transformSourceType=TRANSFORMSOURCE_World, transformType=TRANSFORM_Local}
  n71 = Abs(None=ComponentMask(None=@n70) {b=true})
  n72 = Multiply(A=Clamp(None=ComponentMask(None=@n70) {g=true}) {maxDefault=0, minDefault=-1}) {constB=-1}
  n73 = Saturate(None=Step(X=Add(A=@n71, B=@n72)) {constY=0.9})
  n74 = Multiply(A=Add(A=InverseLinearInterpolate(Value=@n69), B=@n73)) {constB=-1}
  n75 = Multiply(A=Power(Base=Add(A=@n74) {constB=0.3})) {constB=20}
  NamedRerouteDeclaration(None=Saturate(None=Add(A=@n63, B=@n75))) {name=BladeSidesMask, nodeColor=<rgba 0.7578 0 1 1>}
  NamedRerouteDeclaration(None=@n73) {name=AroundEdgeMask, nodeColor=<rgba 0.2656 1 0 1>}
  OneMinus(None=ComponentMask(None=NamedRerouteUsage()) {b=true})
  Sign(None=@n9)
 ]}
