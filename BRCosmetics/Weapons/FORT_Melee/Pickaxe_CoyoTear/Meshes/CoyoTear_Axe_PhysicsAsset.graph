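# Physics asset for the CoyoTear pickaxe, a two-handed melee cosmetic weapon mesh.
# It holds a single body on the Melee_TwoHanded_Handle bone, made of one tilted capsule offset along the handle.
# There are no constraints, so the axe is simulated and collides as one rigid piece.
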PhysicsAsset /BRCosmetics/Weapons/FORT_Melee/Pickaxe_CoyoTear/Meshes/CoyoTear_Axe_PhysicsAsset
bodies {Melee_TwoHanded_Handle [{shapeType=Capsule, length=72.3573, radius=30.3688, center=<0.1204 3.4911 22.1382>, rotation=<rot 4.8935 -90 -0.1222>}]}
constraints []
